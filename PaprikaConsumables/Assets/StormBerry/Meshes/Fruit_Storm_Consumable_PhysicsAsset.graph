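# Physics asset for the StormBerry consumable fruit mesh, defining ragdoll-style bodies for its dynamic stem and leaves.
# A zero-length, large-radius capsule on the handle bone forms the root, with a tiny stem base capsule joined to it by a tight, twist-locked constraint.
# From that base, five twist-locked leaf capsules branch out and an eight-segment stem chain continues, each constrained to its parent.
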PhysicsAsset /PaprikaConsumables/Assets/StormBerry/Meshes/Fruit_Storm_Consumable_PhysicsAsset
bodies {
 dyn_leaf_01 [{shapeType=Capsule, length=2, radius=1.5, shapeName=dyn_leaf_01_capsule, center=<2.8394 0 -1.5449>, rotation=<rot 35.8575 -0.0081 -0.0047>}]
 dyn_leaf_02 [{shapeType=Capsule, length=2, radius=1.5, shapeName=dyn_leaf_02_capsule, center=<2.4041 0 -1.1874>, rotation=<rot 35.8528 -0.0008 -0.0004>}]
 dyn_leaf_03 [{shapeType=Capsule, length=2, radius=1.5, shapeName=dyn_leaf_03_capsule, center=<2.6343 0 -1.0211>, rotation=<rot 35.8589 0.0087 0.005>}]
 dyn_leaf_04 [{shapeType=Capsule, length=2, radius=1.5, shapeName=dyn_leaf_04_capsule, center=<2.4043 0 -1.1874>, rotation=<rot 35.8663 0.0037 0.0021>}]
 dyn_leaf_05 [{shapeType=Capsule, length=2, radius=1.5, shapeName=dyn_leaf_05_capsule, center=<2.8402 0 -1.1895>, rotation=<rot 35.864 -0.0071 -0.0041>}]
 dyn_stem_01 [{shapeType=Capsule, length=0.1, radius=0.1, shapeName=dyn_stem_01_capsule, center=<0.9831 -0.0405 -0.0533>, rotation=<rot 0 86.633 -89.3702>}]
 dyn_stem_02 [{
   shapeType=Capsule
   length=2
   radius=0.75
   shapeName=dyn_stem_02_capsule
   center=<1.171 0 -0.2361>
   rotation=<rot -79.2375 179.9996 -179.9996>
  }]
 dyn_stem_03 [{shapeType=Capsule, length=2, radius=1.5, shapeName=dyn_stem_03_capsule, center=<1.8566 0 0.272>, rotation=<rot -81.6663 -0.0013 0.0013>}]
 dyn_stem_04 [{
   shapeType=Capsule
   length=3
   radius=1
   shapeName=dyn_stem_05_capsule
   center=<2.1229 -0.0002 -0.2929>
   rotation=<rot -86.9102 -0.1834 0.1768>
  }]
 dyn_stem_05 [{
   shapeType=Capsule
   length=1.5
   radius=0.9
   shapeName=dyn_stem_05_capsule
   center=<1.0028 -0.0001 0.0725>
   rotation=<rot -77.921 -0.0646 0.0614>
  }]
 dyn_stem_06 [{shapeType=Capsule, length=2, radius=0.6, shapeName=dyn_stem_06_capsule, center=<1.4997 0 -0.1643>, rotation=<rot 84 179.9992 -179.9991>}]
 dyn_stem_07 [{shapeType=Capsule, length=2, radius=0.5, shapeName=dyn_stem_07_capsule, center=<0.6509 0 -0.0996>, rotation=<rot -81.2733 -0.0004 0.0004>}]
 dyn_stem_08 [{
   shapeType=Capsule
   length=1
   radius=0.5
   shapeName=dyn_stem_08_capsule
   center=<0.3953 0 -0.1206>
   rotation=<rot -73.0347 179.9811 -179.9782>
  }]
 dyn_stem_09 [{shapeType=Capsule, length=1, radius=0.5, shapeName=dyn_stem_09_capsule, rotation=<rot -61.1177 0.0002 -0.0002>}]
 handle [{shapeType=Capsule, length=0, radius=9, center=<0 0 -2.6412>}]
}
constraints [
 {bone1Name=dyn_leaf_05, bone2Name=dyn_stem_01, swing1LimitDegrees=8, swing2LimitDegrees=30, twistLimitDegrees=1, twistMotion=Locked}
 {bone1Name=dyn_leaf_01, bone2Name=dyn_stem_01, swing1LimitDegrees=8, swing2LimitDegrees=30, twistLimitDegrees=1, twistMotion=Locked}
 {bone1Name=dyn_leaf_02, bone2Name=dyn_stem_01, swing1LimitDegrees=8, swing2LimitDegrees=30, twistLimitDegrees=1, twistMotion=Locked}
 {bone1Name=dyn_leaf_03, bone2Name=dyn_stem_01, swing1LimitDegrees=8, swing2LimitDegrees=30, twistLimitDegrees=1, twistMotion=Locked}
 {bone1Name=dyn_leaf_04, bone2Name=dyn_stem_01, swing1LimitDegrees=8, swing2LimitDegrees=30, twistLimitDegrees=1, twistMotion=Locked}
 {bone1Name=dyn_stem_02, bone2Name=dyn_stem_01, swing1LimitDegrees=15, swing2LimitDegrees=15, twistLimitDegrees=1}
 {bone1Name=dyn_stem_03, bone2Name=dyn_stem_02, swing1LimitDegrees=20, swing2LimitDegrees=20, twistLimitDegrees=1}
 {bone1Name=dyn_stem_04, bone2Name=dyn_stem_03, swing1LimitDegrees=20, swing2LimitDegrees=20, twistLimitDegrees=1}
 {bone1Name=dyn_stem_05, bone2Name=dyn_stem_04, swing1LimitDegrees=20, swing2LimitDegrees=20, twistLimitDegrees=1}
 {bone1Name=dyn_stem_06, bone2Name=dyn_stem_05, swing1LimitDegrees=20, swing2LimitDegrees=20, twistLimitDegrees=1}
 {bone1Name=dyn_stem_07, bone2Name=dyn_stem_06, swing1LimitDegrees=20, swing2LimitDegrees=20, twistLimitDegrees=1}
 {bone1Name=dyn_stem_08, bone2Name=dyn_stem_07, swing1LimitDegrees=20, swing2LimitDegrees=20, twistLimitDegrees=1}
 {bone1Name=dyn_stem_09, bone2Name=dyn_stem_08, swing1LimitDegrees=20, swing2LimitDegrees=20, twistLimitDegrees=1}
 {bone1Name=dyn_stem_01, bone2Name=handle, swing1LimitDegrees=2, swing2LimitDegrees=2, twistLimitDegrees=45, twistMotion=Locked}
]
