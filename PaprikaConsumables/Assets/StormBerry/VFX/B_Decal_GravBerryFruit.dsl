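; B_Decal_GravBerryFruit is a decal Blueprint for the StormBerry consumable VFX that shows a GravBerry fruit decal.
; Its construction script creates a dynamic material instance from the small barrel GravFruit decal material variant without caustic motion.
; The instance is stored in the Decal_MID variable and assigned to the Decal component, leaving it available for runtime parameter changes.
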(fn ConstructionScript
  (bind _returnvalue (Rendering|Material|CreateDynamicMaterialInstance "/PaprikaConsumables/Assets/GravBerry_Barrel/MI_Decal_BarrelGravFruit_Small_NoCuasticMotion.MI_Decal_BarrelGravFruit_Small_NoCuasticMotion" "None" "None" :!class /Script/Engine.KismetMaterialLibrary))
  (bind _output_get (Variables|Default|SetDecal_MID _returnvalue))
  (Rendering|Components|Decal|SetDecalMaterial (Variables|Decal|GetDecal) _output_get))
