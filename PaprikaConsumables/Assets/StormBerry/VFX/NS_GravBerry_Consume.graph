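# Niagara burst effect for consuming a GravBerry, with four independent emitters and no user parameters.
# Drips (GPU) and Warp throw sprites outward under drag and gravity, while Splat launches growing sprites upward; all shift from blue toward purple.
# Wave expands two mesh renderers with animated material parameters and time-of-day-scaled opacity.
# Wave also drives a light whose radius follows mesh scale, fading as its color shifts from pink to blue.
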NiagaraSystem /PaprikaConsumables/Assets/StormBerry/VFX/NS_GravBerry_Consume
dependencies {
 usedDataInterfaces [&NiagaraDataInterfaceColorCurve, &NiagaraDataInterfaceCurve, &NiagaraDataInterfaceVector2DCurve]
 usedDynamicInputs [&Vector2DFromFloat, &RotateVectorByQuaternion]
 usedModules [
  &self:SystemSpawnScript.SystemScriptSource.SystemScriptGraph.NiagaraNodeAssignment_1.SetVariables_#1
  &SystemState
 ]
}
emitters {
 Drips {
  simTarget=GPUComputeSim
  emitterUpdateScript [
   &EmitterState()
   &SpawnBurst_Instantaneous(Spawn Count=60, Use Spawn Probability=false, UseLoopCountLimit=false)
  ]
  particleSpawnScript [
   &InitializeParticle(
    Lifetime Max=0.7
    Lifetime Min=0.2
    Lifetime Mode=::NewEnumerator1
    Sprite Rotation Mode=::NewEnumerator2
    Sprite Size Mode=::NewEnumerator3
    Uniform Sprite Size=&RandomRangeFloat(Minimum=3, Maximum=6)
    UsePositionOffset=false
    Write Material Random=true
   )
   &SubUVAnimation(End Frame=3, SubUV Animation Mode:ENiagara_SubUVLookupMode=::NewEnumerator2)
   &ShapeLocation(
    Hemisphere Distribution=<1 0.7>
    MaskSpawnGroup=true
    Offset=$System.SystemPosition
    Offset Mode=::NewEnumerator0
    Override Seed=false
    Sphere Radius=20
    Sphere Surface Distribution=0.15
    UseOwnerScale=true
   )
   &AddVelocity(
    Module.Use Falloff Distance=false
    Origin Offset=&AddVectorToPosition(Position=Linked() {linkedVariable {name=System.SystemPosition, type=Vector3f}}, Vector=<0 0 -7>)
    Origin Offset Coordinate Space=::World
    Override Seed=false
    Velocity Mode=::NewEnumerator1
    Velocity Speed=&RandomRangeFloat(Minimum=250, Maximum=750)
   )
  ]
  particleUpdateScript [
   &ParticleState()
   &Drag(Use Linear Drag=true, Use Rotational Drag=false)
   &GravityForce()
   &SolveForcesAndVelocity(Clamp Velocity=false, Limit Acceleration=false)
   &ScaleColor(
    Scale RGB=&MakeVectorFromLinearColorRGB(Color=&ColorFromCurve(
      ColorCurve=DataInterface() {propertyValues {
        AlphaCurve {keys []}
        BlueCurve {keys [<at 0.0289 0.75 RCIM_Cubic>,<at 0.3797 0.68>,<at 0.9719 0.4>]}
        GreenCurve {keys [<at 0.0289 0.4687 RCIM_Cubic>,<at 0.3797 0.102>,<at 0.9719 0.072>]}
        RedCurve {keys [<at 0.0289 0.075 RCIM_Cubic>,<at 0.3797 0.102>,<at 0.9719 0.2907>]}
       }}
      CurveIndex=&FloatFromCurve(FloatCurve=DataInterface() {propertyValues {Curve {keys [{interpMode=RCIM_Cubic, tangentMode=RCTM_User, time=0.32, value=0}, {interpMode=RCIM_Cubic, tangentMode=RCTM_User, time=0.81, value=1}]}}})
     ))
    ScaleA=true
    ScaleRGB=true
   )
   &ScaleSpriteSize(Uniform Curve Sprite Scale=DataInterface() {propertyValues {Curve {keys [
        {arriveTangent=-0.2878, interpMode=RCIM_Cubic, leaveTangent=-0.2878, tangentMode=RCTM_User, time=0, value=1}
        {arriveTangent=-2.3333, interpMode=RCIM_Cubic, leaveTangent=-2.3333, tangentMode=RCTM_User, time=1, value=0}
       ]}}})
  ]
  rendererClasses [&NiagaraSpriteRendererProperties]
  renderers [{rendererClass=&NiagaraSpriteRendererProperties}]
 }
 Splat {
  emitterUpdateScript [
   &EmitterState()
   &SpawnBurst_Instantaneous(Use Spawn Probability=false, UseLoopCountLimit=false)
  ]
  particleSpawnScript [
   &InitializeParticle(
    Lifetime=0.65
    Sprite Rotation Mode=::NewEnumerator2
    Sprite Size Mode=::NewEnumerator3
    Uniform Sprite Size=100
    UsePositionOffset=false
    Write Material Random=true
   )
   &SubUVAnimation(End Frame=3, SubUV Animation Mode:ENiagara_SubUVLookupMode=::NewEnumerator2)
   &AddVelocity(Rotation Coordinate Space=::World, Velocity=<0 0 150>)
   /SetVariables(Particles.Position=&AddVectorToPosition(Position=$Engine.Emitter.SimulationPosition, Vector=$System.SystemPosition)) {isSetParameters=true}
  ]
  particleUpdateScript [
   &ParticleState()
   &GravityForce()
   &SolveForcesAndVelocity(Clamp Velocity=false, Limit Acceleration=false)
   &ScaleColor(
    Scale Alpha=&FloatFromCurve(FloatCurve=DataInterface() {propertyValues {Curve {keys [
         {arriveTangent=-2.0816, interpMode=RCIM_Cubic, leaveTangent=-2.0816, tangentMode=RCTM_User, time=0, value=0.9}
         {arriveTangent=-0.2382, interpMode=RCIM_Cubic, leaveTangent=-0.2382, tangentMode=RCTM_User, time=1, value=0.3}
        ]}}})
    Scale RGB=&MakeVectorFromLinearColorRGB(Color=&ColorFromCurve(
      ColorCurve=DataInterface() {propertyValues {
        AlphaCurve {keys []}
        BlueCurve {keys [<at 0.008 0.83 RCIM_Cubic>,<at 0.9972 0.6 RCIM_Cubic>]}
        GreenCurve {keys [<at 0.008 0.4565 RCIM_Cubic>,<at 0.9972 0.072 RCIM_Cubic>]}
        RedCurve {keys [<at 0.008 0.083 RCIM_Cubic>,<at 0.9972 0.204 RCIM_Cubic>]}
       }}
      CurveIndex=&FloatFromCurve(FloatCurve=DataInterface() {propertyValues {Curve {keys [{interpMode=RCIM_Cubic, tangentMode=RCTM_User, time=0.18, value=0}, {interpMode=RCIM_Cubic, tangentMode=RCTM_User, time=0.55, value=1}]}}})
     ))
    ScaleA=true
    ScaleRGB=true
   )
   &ScaleSpriteSize(Uniform Curve Sprite Scale=DataInterface() {propertyValues {Curve {keys [
        {arriveTangent=1.7204, interpMode=RCIM_Cubic, leaveTangent=1.7204, tangentMode=RCTM_User, time=0, value=0.25}
        {interpMode=RCIM_Cubic, tangentMode=RCTM_User, time=1, value=1}
       ]}}})
   &DynamicMaterialParameters(
    Module.Param0WriteEnabled1=true
    Module.Param1WriteEnabled1=true
    Module.Param2WriteEnabled1=false
    Module.Param3WriteEnabled1=false
    Write Parameter Index 0=false
    Write Parameter Index 1=true
   )
  ]
  rendererClasses [&NiagaraSpriteRendererProperties]
  renderers [{rendererClass=&NiagaraSpriteRendererProperties}]
 }
 Warp {
  emitterUpdateScript [
   &EmitterState()
   &SpawnBurst_Instantaneous(Spawn Count=6, Use Spawn Probability=false, UseLoopCountLimit=false)
  ]
  particleSpawnScript [
   &InitializeParticle(
    Lifetime Max=0.78
    Lifetime Min=0.43
    Lifetime Mode=::NewEnumerator1
    Sprite Rotation Mode=::NewEnumerator2
    Sprite Size Mode=::NewEnumerator1
    Uniform Sprite Size Max=120
    Uniform Sprite Size Min=70
    UsePositionOffset=false
    Write Material Random=true
   )
   &SubUVAnimation(End Frame=3, SubUV Animation Mode:ENiagara_SubUVLookupMode=::NewEnumerator2)
   &ShapeLocation(
    Hemisphere Distribution=<1 0.7>
    MaskSpawnGroup=true
    Offset=$System.SystemPosition
    Offset Mode=::NewEnumerator0
    Override Seed=false
    Sphere Radius=5
    Sphere Surface Distribution=0.15
    UseOwnerScale=true
   )
   &AddVelocity(
    Module.Use Falloff Distance=false
    Origin Offset=&AddVectorToPosition(Position=Linked() {linkedVariable {name=System.SystemPosition, type=Vector3f}}, Vector=<0 0 -3>)
    Origin Offset Coordinate Space=::World
    Override Seed=false
    Velocity Mode=::NewEnumerator1
    Velocity Speed=&RandomRangeFloat(Minimum=200, Maximum=250)
   )
  ]
  particleUpdateScript [
   &ParticleState()
   &Drag(Drag=2, Use Linear Drag=true, Use Rotational Drag=false)
   &GravityForce()
   &SolveForcesAndVelocity(Clamp Velocity=false, Limit Acceleration=false)
   &ScaleColor(
    Scale Alpha=&FloatFromCurve(FloatCurve=DataInterface() {propertyValues {Curve {keys [{interpMode=RCIM_Cubic, tangentMode=RCTM_User, time=0, value=1}, {interpMode=RCIM_Cubic, tangentMode=RCTM_User, time=1, value=0}]}}})
    Scale RGB=&MakeVectorFromLinearColorRGB(Color=&ColorFromCurve(
      ColorCurve=DataInterface() {propertyValues {
        AlphaCurve {keys []}
        BlueCurve {keys [<at 0 0.75 RCIM_Cubic>,<at 0.4023 0.7>,<at 0.9719 0.5>]}
        GreenCurve {keys [<at 0 0.5327 RCIM_Cubic>,<at 0.4023 0.1161>,<at 0.9719 0.05>]}
        RedCurve {keys [<at 0 0 RCIM_Cubic>,<at 0.4023 0.063>,<at 0.9719 0.35>]}
       }}
      CurveIndex=&FloatFromCurve(FloatCurve=DataInterface() {propertyValues {Curve {keys [{interpMode=RCIM_Cubic, tangentMode=RCTM_User, time=0.15, value=0}, {interpMode=RCIM_Cubic, tangentMode=RCTM_User, time=0.7, value=1}]}}})
     ))
    ScaleA=true
    ScaleRGB=true
   )
   &ScaleSpriteSize(Uniform Curve Sprite Scale=DataInterface() {propertyValues {Curve {keys [
        {arriveTangent=1.2875, interpMode=RCIM_Cubic, leaveTangent=1.2875, tangentMode=RCTM_User, time=0, value=0.3}
        {interpMode=RCIM_Cubic, tangentMode=RCTM_User, time=1, value=1}
       ]}}})
   &DynamicMaterialParameters(
    Module.Param0WriteEnabled1=true
    Module.Param1WriteEnabled1=true
    Module.Param2WriteEnabled1=true
    Module.Param3WriteEnabled1=true
    Write Parameter Index 0=false
    Write Parameter Index 1=true
   )
  ]
  rendererClasses [&NiagaraSpriteRendererProperties]
  renderers [{rendererClass=&NiagaraSpriteRendererProperties}]
 }
 Wave {
  emitterUpdateScript [
   &EmitterState(Life Cycle Mode=::NewEnumerator1, Loop Behavior=::NewEnumerator1, Loop Delay=0.015, Loop Duration=1, UseLoopDelay=false)
   &SpawnBurst_Instantaneous(Use Spawn Probability=false, UseLoopCountLimit=false)
  ]
  particleSpawnScript [
   &InitializeParticle(
    Color=<rgba 1 1 1 0.2>
    Lifetime=0.5
    Mesh Scale Mode=::NewEnumerator3
    Mesh Uniform Scale=4
    UsePositionOffset=false
    Write Material Random=true
   )
   &InitialMeshOrientation(
    Mesh Orientation Mode=::NewEnumerator2
    Orientation Vector=<0 0 -1>
    Rotation=&RandomRangeVector(Minimum=<0 0 -1>, Maximum=<0 0 1>)
    Use Rotation Vector=true
   )
   /SetVariables(Particles.Position=&AddVectorToPosition(
     Position=&AddVectorToPosition(Position=$Engine.Emitter.SimulationPosition, Vector=<0 0 0>)
     Vector=$System.SystemPosition
    )) {isSetParameters=true}
  ]
  particleUpdateScript [
   &ParticleState()
   &ScaleColor(
    Scale Alpha=&FloatFromCurve(
     FloatCurve=DataInterface() {propertyValues {Curve {keys [
         {arriveTangent=27.4578, interpMode=RCIM_Cubic, leaveTangent=27.4578, tangentMode=RCTM_User, time=0, value=0}
         {arriveTangent=-0.022, interpMode=RCIM_Cubic, leaveTangent=-0.022, tangentMode=RCTM_User, time=0.1, value=1}
         {interpMode=RCIM_Cubic, tangentMode=RCTM_User, time=1, value=0}
        ]}}}
     Scale Curve=&FNTimeOfDayScalar(Night Scale=0.7, Morning Scale=0.5, Evening Scale=0.5)
    )
    Scale RGB=&MakeVectorFromLinearColorRGB(Color=&ColorFromCurve(
      ColorCurve=DataInterface() {propertyValues {
        AlphaCurve {keys []}
        BlueCurve {keys [<at 0.029 1 RCIM_Cubic>,<at 0.1185 1>,<at 0.7455 0.56 RCIM_Cubic>]}
        GreenCurve {keys [<at 0.029 0.8333 RCIM_Cubic>,<at 0.1185 0.1>,<at 0.7455 0.0392 RCIM_Cubic>]}
        RedCurve {keys [<at 0.029 0 RCIM_Cubic>,<at 0.1185 0.1289>,<at 0.7455 0.3864 RCIM_Cubic>]}
       }}
      CurveIndex=&FloatFromCurve(FloatCurve=DataInterface() {propertyValues {Curve {keys [{interpMode=RCIM_Cubic, tangentMode=RCTM_User, time=0.2, value=0}, {interpMode=RCIM_Cubic, tangentMode=RCTM_User, time=0.4, value=1}]}}})
     ))
    ScaleA=true
    ScaleRGB=true
   ) {module=ScaleColor002}
   &DynamicMaterialParameters(
    Index 0 Param 1=&FloatFromCurve(FloatCurve=DataInterface() {propertyValues {Curve {keys [{arriveTangent=-1.1757, interpMode=RCIM_Cubic, leaveTangent=-1.1757, tangentMode=RCTM_User, time=0, value=0.5}, <at 1 0>]}}})
    Index 0 Param 2=&FloatFromCurve(FloatCurve=DataInterface() {propertyValues {Curve {keys [{arriveTangent=-0.7726, interpMode=RCIM_Cubic, leaveTangent=-0.7726, tangentMode=RCTM_User, time=0, value=1}, <at 1 0.7>]}}})
    Index 0 Param 3=&FloatFromCurve(FloatCurve=DataInterface() {propertyValues {Curve {keys [{interpMode=RCIM_Cubic, tangentMode=RCTM_User, time=0, value=0.7}, {interpMode=RCIM_Cubic, tangentMode=RCTM_User, time=0.8, value=0.25}]}}})
    Module.Param0WriteEnabled=true
    Module.Param1WriteEnabled=true
    Module.Param2WriteEnabled=true
    Module.Param3WriteEnabled=false
   )
   &ScaleMeshSize(Scale Factor=&VectorFromFloat(Value=&FloatFromCurve(FloatCurve=DataInterface() {propertyValues {Curve {keys [
          {arriveTangent=9.0273, interpMode=RCIM_Cubic, leaveTangent=9.0273, tangentMode=RCTM_User, time=0, value=0.1}
          {arriveTangent=1.2987, interpMode=RCIM_Cubic, leaveTangent=1.2987, tangentMode=RCTM_User, time=0.1, value=0.55}
          <at 1 1>
         ]}}})))
   /SetVariables(Particles.Intensity:NiagaraFloat=&FloatFromCurve(
     FloatCurve=DataInterface() {propertyValues {Curve {keys [{interpMode=RCIM_Cubic, tangentMode=RCTM_User, time=0.2, value=1}, {interpMode=RCIM_Cubic, tangentMode=RCTM_User, time=1, value=0}]}}}
     Scale Curve=5
    )) {isSetParameters=true}
   &FN_Light_Attributes(
    Color=&MultiplyLinearColor(
     Linear Color=&LerpLinearColors(
      StartColor=<rgba 1 0.1 0.418 1>
      EndColor=<rgba 0 0.2294 0.974 0.5>
      LerpFactor=&FloatFromCurve(FloatCurve=DataInterface() {propertyValues {Curve {keys [{interpMode=RCIM_Cubic, tangentMode=RCTM_User, time=0.25, value=0}, {interpMode=RCIM_Cubic, tangentMode=RCTM_User, time=0.5, value=1}]}}})
     )
     Scale Factor=&MakeFloat4FromVectorAndFloat("Vector (RGB)"=<1 1 1>, "Float (Alpha)"=$Particles.Intensity)
    )
    Create Color Attribute=true
    Create Exponent Attribute=true
    Create Radius Attribute=true
    Exponent=1.5
    Radius=&Multiply_Float(A=&MakeFloatFromVector(Vector=$Particles.Scale, Channel=::NewEnumerator0), B=75)
   )
  ]
  rendererClasses [&NiagaraMeshRendererProperties, &NiagaraLightRendererProperties]
  renderers [
   {rendererClass=&NiagaraMeshRendererProperties, rendererIndex=0}
   {rendererClass=&NiagaraMeshRendererProperties, rendererIndex=1}
   {rendererClass=&NiagaraLightRendererProperties, rendererIndex=2}
  ]
 }
}
summary {
 systemName=NS_GravBerry_Consume
 emitters [
  {emitterName=Wave, rendererClasses [&NiagaraMeshRendererProperties, &NiagaraLightRendererProperties]}
  {emitterName=Splat, rendererClasses [&NiagaraSpriteRendererProperties]}
  {emitterName=Warp, rendererClasses [&NiagaraSpriteRendererProperties]}
  {emitterName=Drips, simTarget=GPUComputeSim, rendererClasses [&NiagaraSpriteRendererProperties]}
 ]
 userVariables []
}
