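# Niagara one-shot-style splash for the GravBerry consumable, named for use on the player, with two blue-to-purple sprite emitters.
# Warp bursts and streams large sprites from a cylinder surface that sweeps downward over the loop, growing, stretching and fading them while curves drive dynamic material parameters.
# Drips is a GPU burst of small droplets thrown from a cylinder, slowed by drag, pulled by gravity, flashing bright, then shrinking away.
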NiagaraSystem /PaprikaConsumables/Assets/StormBerry/VFX/NS_GravBerry_PlayerSplash
dependencies {
 usedDataInterfaces [&NiagaraDataInterfaceCurve, &NiagaraDataInterfaceColorCurve, &NiagaraDataInterfaceVector2DCurve]
 usedDynamicInputs [&Vector2DFromFloat, &RotateVectorByQuaternion]
 usedModules [
  &self:SystemSpawnScript.SystemScriptSource.SystemScriptGraph.NiagaraNodeAssignment_1.SetVariables_#1
  &SystemState
 ]
}
emitters {
 Drips {
  simTarget=GPUComputeSim
  emitterUpdateScript [
   &EmitterState()
   &SpawnRate(SpawnRate=5, Use Spawn Probability=true) {enabled=false}
   &SpawnBurst_Instantaneous(Spawn Count=10, Use Spawn Probability=false, UseLoopCountLimit=false)
  ]
  particleSpawnScript [
   &InitializeParticle(
    Color=&RandomRangeLinearColor(Minimum=<rgba 0.063 0.2753 0.7 1>, Maximum=<rgba 0.32 0.12 0.6 1>)
    Lifetime Max=0.5
    Lifetime Min=0.3
    Lifetime Mode=::NewEnumerator1
    Sprite Rotation Mode=::NewEnumerator2
    Sprite Size Mode=::NewEnumerator3
    Uniform Sprite Size=&RandomRangeFloat(Minimum=3, Maximum=4)
    UsePositionOffset=false
    Write Material Random=true
   )
   &SubUVAnimation(End Frame=3, SubUV Animation Mode:ENiagara_SubUVLookupMode=::NewEnumerator2)
   &ShapeLocation(
    Cylinder Height=60
    Cylinder Radius=17
    MaskSpawnGroup=true
    Offset=$System.SystemPosition
    Offset Mode=::NewEnumerator0
    Override Seed=false
    Shape Primitive=::NewEnumerator1
    Surface Only=true
    Surface Only Band Thickness=0.8
    UseOwnerScale=true
   )
   &AddVelocity(
    Module.Use Falloff Distance=false
    Origin Offset=&AddVectorToPosition(Position=Linked() {linkedVariable {name=System.SystemPosition, type=Vector3f}}, Vector=<0 0 -20>)
    Override Seed=false
    Velocity Mode=::NewEnumerator1
    Velocity Speed=&RandomRangeFloat(Minimum=100, Maximum=200)
   )
   &AddVelocity(
    Velocity=$Particles.ShapeLocation.ShapeNormal
    Velocity Speed Scale=&RandomRangeFloat(Minimum=200, Maximum=400)
   ) {enabled=false, module=AddVelocity001}
  ]
  particleUpdateScript [
   &ParticleState()
   &Drag(Use Linear Drag=true, Use Rotational Drag=false)
   &GravityForce()
   &SolveForcesAndVelocity(Clamp Velocity=false, Limit Acceleration=false)
   &ScaleColor(
    Scale RGB=&MakeVectorFromLinearColorRGB(Color=&ColorFromCurve(ColorCurve=DataInterface() {propertyValues {
        AlphaCurve {keys []}
        BlueCurve {keys [<at 0.0958 3 RCIM_Cubic>,<at 0.5407 1 RCIM_Cubic>]}
        GreenCurve {keys [<at 0.0958 3 RCIM_Cubic>,<at 0.5407 1 RCIM_Cubic>]}
        RedCurve {keys [<at 0.0958 3 RCIM_Cubic>,<at 0.5407 1 RCIM_Cubic>]}
       }}))
    ScaleA=false
    ScaleRGB=true
   )
   &ScaleSpriteSize(Uniform Curve Sprite Scale=DataInterface() {propertyValues {Curve {keys [
        {arriveTangent=-0.2878, interpMode=RCIM_Cubic, leaveTangent=-0.2878, tangentMode=RCTM_User, time=0, value=1}
        {arriveTangent=-2.3333, interpMode=RCIM_Cubic, leaveTangent=-2.3333, tangentMode=RCTM_User, time=1, value=0}
       ]}}})
  ]
  rendererClasses [&NiagaraSpriteRendererProperties]
  renderers [{rendererClass=&NiagaraSpriteRendererProperties}]
 }
 Warp {
  emitterUpdateScript [
   &EmitterState()
   &SpawnRate(SpawnRate=60, Use Spawn Probability=true)
   &SpawnBurst_Instantaneous(
    Spawn Count=&RandomRangeInt(Minimum=6, Maximum=10, Recalculate Random Each Loop=false)
    Use Spawn Probability=false
    UseLoopCountLimit=false
   )
  ]
  particleSpawnScript [
   &InitializeParticle(
    Color=&RandomRangeLinearColor(Minimum=<rgba 0.0448 0.2432 0.64 1>, Maximum=<rgba 0.24 0.08 0.4 1>)
    Lifetime Max=0.5
    Lifetime Min=0.36
    Lifetime Mode=::NewEnumerator1
    Mass Mode=::NewEnumerator3
    Sprite Rotation Mode=::NewEnumerator2
    Sprite Size=&Multiply_Vector2D_ByFloat(Float=&RandomRangeFloat(Minimum=44, Maximum=75), Vector2D=<0.8 1>)
    Sprite Size Mode=::NewEnumerator4
    UsePositionOffset=false
    Write Material Random=true
   )
   &SubUVAnimation(End Frame=3, SubUV Animation Mode:ENiagara_SubUVLookupMode=::NewEnumerator2)
   &ShapeLocation(
    Cylinder Height=60
    Cylinder Radius=20
    MaskSpawnGroup=true
    Offset=&AddVectorToPosition(
     Position=Linked() {linkedVariable {name=System.SystemPosition, type=Vector3f}}
     Vector=&MakeVector(
      X=0
      Y=0
      Z=&FloatFromCurve(FloatCurve=DataInterface() {propertyValues {Curve {keys [<at 0 0>,<at 1 1>]}}}, CurveIndex=$Emitter.NormalizedLoopAge, Scale Curve=-160)
     )
    )
    Offset Mode=::NewEnumerator0
    Override Seed=false
    Shape Primitive=::NewEnumerator1
    Surface Only=true
    Surface Only Band Thickness=0.8
    UseOwnerScale=true
   )
   &AddVelocity(Velocity=$Particles.ShapeLocation.ShapeNormal) {module=AddVelocity002}
   &AddVelocity(
    Module.Use Falloff Distance=false
    Origin Offset=&AddVectorToPosition(Position=Linked() {linkedVariable {name=System.SystemPosition, type=Vector3f}}, Vector=<0 0 4>)
    Override Seed=false
    Velocity Mode=::NewEnumerator1
    Velocity Speed=&RandomRangeFloat(Minimum=40, Maximum=70)
   ) {module=AddVelocity001}
   &CameraOffset(Camera Offset Amount=10)
  ]
  particleUpdateScript [
   &ParticleState()
   &Drag(Use Linear Drag=true, Use Rotational Drag=false)
   &GravityForce(Gravity=<0 0 -600>)
   &SolveForcesAndVelocity(Clamp Velocity=false, Limit Acceleration=false)
   &ScaleColor(
    Scale Alpha=&FloatFromCurve(FloatCurve=DataInterface() {propertyValues {Curve {keys [{interpMode=RCIM_Cubic, tangentMode=RCTM_User, time=0, value=1}, {interpMode=RCIM_Cubic, tangentMode=RCTM_User, time=1, value=0}]}}})
    Scale RGB=&MakeVectorFromLinearColorRGB(Color=&ColorFromCurve(ColorCurve=DataInterface() {propertyValues {
        AlphaCurve {keys []}
        BlueCurve {keys [<at 0.2098 2 RCIM_Cubic>,<at 0.3817 1 RCIM_Cubic>]}
        GreenCurve {keys [<at 0.2098 2 RCIM_Cubic>,<at 0.3817 1 RCIM_Cubic>]}
        RedCurve {keys [<at 0.2098 2 RCIM_Cubic>,<at 0.3817 1 RCIM_Cubic>]}
       }}))
    ScaleA=true
    ScaleRGB=false
   )
   &ScaleSpriteSize(Uniform Curve Sprite Scale=DataInterface() {propertyValues {Curve {keys [
        {arriveTangent=1.2875, interpMode=RCIM_Cubic, leaveTangent=1.2875, tangentMode=RCTM_User, time=0, value=0.3}
        {interpMode=RCIM_Cubic, tangentMode=RCTM_User, time=1, value=1}
       ]}}})
   &ScaleSpriteSize(
    Non-Uniform Curve Sprite Scale=DataInterface() {propertyValues {
      XCurve {keys [{interpMode=RCIM_Cubic, tangentMode=RCTM_User, time=0, value=1}, {interpMode=RCIM_Cubic, tangentMode=RCTM_User, time=1, value=1}]}
      YCurve {keys [{interpMode=RCIM_Cubic, tangentMode=RCTM_User, time=0, value=1}, {interpMode=RCIM_Cubic, tangentMode=RCTM_User, time=0.7, value=1.5}]}
     }}
    Scale Sprite Size Mode=::NewEnumerator3
   ) {module=ScaleSpriteSize001}
   &DynamicMaterialParameters(
    Module.Param0WriteEnabled1=true
    Module.Param1WriteEnabled1=true
    Module.Param2WriteEnabled1=true
    Module.Param3WriteEnabled1=true
    Write Parameter Index 0=false
    Write Parameter Index 1=true
   ) {enabled=false}
   &DynamicMaterialParameters(
    Index 0 Param 2=&FloatFromCurve(FloatCurve=DataInterface() {propertyValues {Curve {keys [{interpMode=RCIM_Cubic, tangentMode=RCTM_User, time=0, value=1}, {interpMode=RCIM_Cubic, tangentMode=RCTM_User, time=0.34, value=0}]}}})
    Index 0 Param 4=&FloatFromCurve(FloatCurve=DataInterface() {propertyValues {Curve {keys [{arriveTangent=0.4751, interpMode=RCIM_Cubic, leaveTangent=0.4751, tangentMode=RCTM_User, time=0.1, value=0}, <at 1 1 RCIM_Cubic>]}}})
    Module.Index0Param0Write=true
    Module.Index0Param1Write=true
    Module.Index0Param2Write=false
    Module.Index0Param3Write=true
   ) {module=DynamicMaterialParameters001}
  ]
  rendererClasses [&NiagaraSpriteRendererProperties]
  renderers [{rendererClass=&NiagaraSpriteRendererProperties}]
 }
}
summary {
 systemName=NS_GravBerry_PlayerSplash
 emitters [
  {emitterName=Warp, rendererClasses [&NiagaraSpriteRendererProperties]}
  {emitterName=Drips, simTarget=GPUComputeSim, rendererClasses [&NiagaraSpriteRendererProperties]}
 ]
 userVariables []
}
